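; PCG loop subgraph for the Draw Spline editor mode's linear grammar that applies forward-axis, offset and rotation corrections to points.
; ForwardAxis selects a scale swizzle for Y or Z, then ForwardAxisRotation is combined with point rotation.
; Negated, divided Size, masked by ForwardAxisVector, forms a pivot offset, to which sign-adjusted LocalOffset is added.
; This offset is scaled, rotated and added to position, then LocalRotation is applied.
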(graph LinearGrammar_ApplyOffsetsLoop
  (bind AttributeMathsOp_0 (Attribute%20Maths%20Op :InputSource1 $Scale :InputSource2 $Scale.yxz :Operation Set))
  (bind AttributeMathsOp_1 (Attribute%20Maths%20Op :InputSource1 $Scale :InputSource2 $Scale.zyx :Operation Set))
  (bind ExtractAttribute_0 (Extract%20Attribute%20at%20Index :bForceOutputAttributeToBeInElementsDomain true :InputSource ForwardAxis))
  (bind Select%20%28Multi%29_0 (Select%20%28Multi%29 :EnumSelection (struct :Class /Script/CoreUObject.EAxis) :SelectionMode Enum))
  (bind AttributeRotatorOp_4 (Attribute%20Rotator%20Op :InputSource1 ForwardAxisRotation :InputSource2 $Rotation :OutputTarget $Rotation))
  (bind AttributeMathsOp_9 (Attribute%20Maths%20Op :InputSource1 Size :Operation Negate :OutputTarget PivotOffset))
  (bind AttributeMathsOp_10 (Attribute%20Maths%20Op :InputSource1 PivotOffset :Operation Divide))
  (bind AttributeMathsOp_5 (Attribute%20Maths%20Op :InputSource1 PivotOffset :InputSource2 ForwardAxisVector :Operation Multiply)) ; Only Consider Offset X (Along Spline)
  (bind AttributeMathsOp_8 (Attribute%20Maths%20Op :InputSource1 ForwardAxisVector :OutputTarget ForwardAxisVectorSign))
  (bind AttributeMathsOp_11 (Attribute%20Maths%20Op :InputSource1 ForwardAxisVectorSign :Operation Sign :OutputTarget ForwardAxisVectorSign))
  (bind AttributeMathsOp_12 (Attribute%20Maths%20Op :InputSource1 LocalOffset :InputSource2 ForwardAxisVectorSign :Operation Multiply))
  (bind AttributeMathsOp_6 (Attribute%20Maths%20Op :InputSource1 PivotOffset :InputSource2 LocalOffset))
  (bind AttributeMathsOp_2 (Attribute%20Maths%20Op :InputSource1 PivotOffset :InputSource2 $Scale :Operation Multiply)) ; PivotOffset needs to be \nscaled too
  (bind AttributeVectorOp_1 (Attribute%20Vector%20Op :InputSource1 PivotOffset :InputSource2 $Transform :Operation TransformDirection)) ; And rotated
  (bind AttributeMathsOp_4 (Attribute%20Maths%20Op :InputSource1 $Position :InputSource2 PivotOffset)) ; And added to the position
  (bind AttributeRotatorOp_3 (Attribute%20Rotator%20Op :InputSource1 LocalRotation :InputSource2 $Rotation :OutputTarget $Rotation))
  (connect DefaultInputNode In AttributeMathsOp_0 InA)
  (connect DefaultInputNode In AttributeMathsOp_0 InB)
  (connect DefaultInputNode In AttributeMathsOp_1 InA)
  (connect DefaultInputNode In AttributeMathsOp_1 InB)
  (connect AttributeMathsOp_6 Out AttributeMathsOp_2 InA)
  (connect AttributeMathsOp_6 Out AttributeMathsOp_2 InB)
  (connect AttributeVectorOp_1 Out AttributeMathsOp_4 InA)
  (connect AttributeVectorOp_1 Out AttributeMathsOp_4 InB)
  (connect AttributeMathsOp_10 Out AttributeMathsOp_5 InA)
  (connect AttributeMathsOp_10 Out AttributeMathsOp_5 InB)
  (connect AttributeMathsOp_5 Out AttributeMathsOp_6 InA)
  (connect AttributeMathsOp_12 Out AttributeMathsOp_6 InB)
  (connect AttributeMathsOp_5 Out AttributeMathsOp_8 InA)
  (connect AttributeRotatorOp_4 Out AttributeMathsOp_9 In)
  (connect AttributeMathsOp_9 Out AttributeMathsOp_10 InA)
  (connect AttributeMathsOp_8 Out AttributeMathsOp_11 In)
  (connect AttributeMathsOp_5 Out AttributeMathsOp_12 InA)
  (connect AttributeMathsOp_11 Out AttributeMathsOp_12 InB)
  (connect AttributeMathsOp_4 Out AttributeRotatorOp_3 InA)
  (connect AttributeMathsOp_4 Out AttributeRotatorOp_3 InB)
  (connect Select%20%28Multi%29_0 Out AttributeRotatorOp_4 InA)
  (connect Select%20%28Multi%29_0 Out AttributeRotatorOp_4 InB)
  (connect AttributeMathsOp_2 Out AttributeVectorOp_1 In)
  (connect AttributeMathsOp_2 Out AttributeVectorOp_1 Transform)
  (connect DefaultInputNode In ExtractAttribute_0 In)
  (connect DefaultInputNode In Select%20%28Multi%29_0 Default)
  (connect DefaultInputNode In Select%20%28Multi%29_0 None)
  (connect ExtractAttribute_0 Out Select%20%28Multi%29_0 Value)
  (connect DefaultInputNode In Select%20%28Multi%29_0 X)
  (connect AttributeMathsOp_0 Out Select%20%28Multi%29_0 Y)
  (connect AttributeMathsOp_1 Out Select%20%28Multi%29_0 Z)
  (output Out AttributeRotatorOp_3))
