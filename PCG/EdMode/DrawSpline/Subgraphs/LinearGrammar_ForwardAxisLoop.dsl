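; PCG subgraph, apparently a loop body for the DrawSpline linear grammar, that derives forward-axis vector and rotation attributes from the ForwardAxis enum.
; Enum-driven selects map X, Y or Z to ForwardAxisVector and ForwardAxisRotation, and a default PivotOffset vector is also added.
; Attribute filters then negate the vector for InvertForwardAxis and swap in an alternate per-axis rotation for ReverseOrientation.
; Two rotator constants are left unconnected.
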(graph LinearGrammar_ForwardAxisLoop
  (bind CreateConstant_0 (Create%20Constant :AttributeTypes (struct :Type Vector :VectorValue (1.0 0.0 0.0))))
  (bind CreateConstant_1 (Create%20Constant :AttributeTypes (struct :Type Vector :VectorValue (0.0 1.0 0.0))))
  (bind CreateConstant_2 (Create%20Constant :AttributeTypes (struct :Type Vector :VectorValue (0.0 0.0 1.0))))
  (bind ExtractAttribute_0 (Extract%20Attribute%20at%20Index :bForceOutputAttributeToBeInElementsDomain true :InputSource ForwardAxis))
  (bind Select%20%28Multi%29_0 (Select%20%28Multi%29 :EnumSelection (struct :Class /Script/CoreUObject.EAxis) :SelectionMode Enum))
  (bind AddAttribute_0 (Add%20Attribute :OutputTarget ForwardAxisVector))
  (bind CreateConstant_3 (Create%20Constant :AttributeTypes (struct :Type Rotator :VectorValue (1.0 0.0 0.0))))
  (bind CreateConstant_5 (Create%20Constant :AttributeTypes (struct :Type Rotator :VectorValue (0.0 0.0 1.0) :RotatorValue (struct :Yaw 9e+01 :Roll 2.7e+02))))
  (bind CreateConstant_9 (Create%20Constant :AttributeTypes (struct :Type Rotator :VectorValue (0.0 1.0 0.0) :RotatorValue (struct :Yaw 2.7e+02))))
  (bind Select%20%28Multi%29_1 (Select%20%28Multi%29 :EnumSelection (struct :Class /Script/CoreUObject.EAxis) :SelectionMode Enum))
  (bind AddAttribute_1 (Add%20Attribute :OutputTarget ForwardAxisRotation))
  (bind AddAttribute_2 (Add%20Attribute :AttributeTypes (struct :Type Vector) :OutputTarget PivotOffset))
  (bind AttributeFilter_0 (Filter%20Attribute%20Elements :AttributeTypes (struct :Type Boolean) :bGenerateOutputDataEvenIfEmpty false :bUseConstantThreshold true :Operator Equal :TargetAttribute InvertForwardAxis :ThresholdAttribute @Last))
  (bind AttributeMathsOp_1 (Attribute%20Maths%20Op :InputSource1 ForwardAxisVector :Operation Negate))
  (bind CreateConstant_4 (Create%20Constant :AttributeTypes (struct :Type Rotator :VectorValue (0.0 1.0 0.0) :RotatorValue (struct :Pitch 9e+01 :Yaw 2.7e+02))))
  (bind CreateConstant_6 (Create%20Constant :AttributeTypes (struct :Type Rotator :VectorValue (1.0 0.0 0.0) :RotatorValue (struct :Yaw 1.8e+02))))
  (bind CreateConstant_7 (Create%20Constant :AttributeTypes (struct :Type Rotator :VectorValue (0.0 1.0 0.0) :RotatorValue (struct :Yaw 9e+01))))
  (bind CreateConstant_8 (Create%20Constant :AttributeTypes (struct :Type Rotator :VectorValue (0.0 0.0 1.0) :RotatorValue (struct :Yaw 2.7e+02 :Roll 2.7e+02))))
  (bind Select%20%28Multi%29_2 (Select%20%28Multi%29 :EnumSelection (struct :Class /Script/CoreUObject.EAxis) :SelectionMode Enum))
  (bind AttributeFilter_1 (Filter%20Attribute%20Elements :AttributeTypes (struct :Type Boolean) :bGenerateOutputDataEvenIfEmpty false :bUseConstantThreshold true :Operator Equal :TargetAttribute ReverseOrientation :ThresholdAttribute @Last))
  (bind AttributeMathsOp_2 (Attribute%20Maths%20Op :InputSource1 ForwardAxisRotation :Operation Set))
  (bind CreateConstant_10 (Create%20Constant :AttributeTypes (struct :Type Rotator :VectorValue (0.0 1.0 0.0) :RotatorValue (struct :Pitch 9e+01 :Yaw 2.7e+02))))
  (connect Select%20%28Multi%29_0 Out AddAttribute_0 Attributes)
  (connect DefaultInputNode In AddAttribute_0 In)
  (connect Select%20%28Multi%29_1 Out AddAttribute_1 Attributes)
  (connect AddAttribute_0 Out AddAttribute_1 In)
  (connect AddAttribute_1 Out AddAttribute_2 In)
  (connect AddAttribute_2 Out AttributeFilter_0 In)
  (connect AttributeFilter_0 InsideFilter AttributeFilter_1 In)
  (connect AttributeMathsOp_1 Out AttributeFilter_1 In)
  (connect AttributeFilter_0 OutsideFilter AttributeMathsOp_1 In)
  (connect AttributeFilter_1 OutsideFilter AttributeMathsOp_2 InA)
  (connect Select%20%28Multi%29_2 Out AttributeMathsOp_2 InB)
  (connect DefaultInputNode In ExtractAttribute_0 In)
  (connect CreateConstant_0 Out Select%20%28Multi%29_0 Default)
  (connect CreateConstant_0 Out Select%20%28Multi%29_0 None)
  (connect ExtractAttribute_0 Out Select%20%28Multi%29_0 Value)
  (connect CreateConstant_0 Out Select%20%28Multi%29_0 X)
  (connect CreateConstant_1 Out Select%20%28Multi%29_0 Y)
  (connect CreateConstant_2 Out Select%20%28Multi%29_0 Z)
  (connect CreateConstant_3 Out Select%20%28Multi%29_1 Default)
  (connect CreateConstant_3 Out Select%20%28Multi%29_1 None)
  (connect ExtractAttribute_0 Out Select%20%28Multi%29_1 Value)
  (connect CreateConstant_3 Out Select%20%28Multi%29_1 X)
  (connect CreateConstant_9 Out Select%20%28Multi%29_1 Y)
  (connect CreateConstant_5 Out Select%20%28Multi%29_1 Z)
  (connect CreateConstant_6 Out Select%20%28Multi%29_2 Default)
  (connect CreateConstant_6 Out Select%20%28Multi%29_2 None)
  (connect ExtractAttribute_0 Out Select%20%28Multi%29_2 Value)
  (connect CreateConstant_6 Out Select%20%28Multi%29_2 X)
  (connect CreateConstant_7 Out Select%20%28Multi%29_2 Y)
  (connect CreateConstant_8 Out Select%20%28Multi%29_2 Z)
  (output Out AttributeFilter_1.InsideFilter)
  (output Out AttributeMathsOp_2))
