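; PCG loop subgraph for the linear-grammar Draw Spline editor mode tool that spawns spline meshes using each input's StaticMesh, ForwardAxis and ReverseOrientation.
; Reversed inputs get their spline direction flipped.
; The adjusted, clamped forward axis drives a branch that adds an orientation-dependent local rotation to match the Y-forward module standard.
; SetSplineMeshRoll selects between a plain spawner and one overriding start roll, end roll and up direction from data attributes.
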(graph LinearGrammar_SpawnSplineMeshLoop
  (params (SetSplineMeshRoll :type Bool) (NewProperty :type Vector :default (1.0 0.0 0.0)))
  (bind ExtractAttribute_0 (Extract%20Attribute%20at%20Index :bForceOutputAttributeToBeInElementsDomain true :InputSource ForwardAxis))
  (bind AttributeMathsOp_0 (Attribute%20Maths%20Op :Operation Subtract))
  (bind AttributeMathsOp_1 (Attribute%20Maths%20Op :Operation Clamp))
  (bind AttributeCompareOp_0 (Attribute%20Compare%20Op))
  (bind ExtractAttribute_1 (Extract%20Attribute%20at%20Index :bForceOutputAttributeToBeInElementsDomain true :InputSource ReverseOrientation))
  (bind Branch_2 (Branch))
  (bind SplineDirection_1 (Spline%20Direction))
  (bind Branch_1 (Branch))
  (bind Branch_3 (Branch))
  (bind Branch_0 (Branch :bOutputToB-param SetSplineMeshRoll))
  (bind AttributeMathsOp_4 (Attribute%20Maths%20Op :InputSource1 @Data.RollDegrees))
  (bind AttributeMathsOp_5 (Attribute%20Maths%20Op :InputSource1 @Data.RollDegreesEnd))
  (bind SpawnSplineMesh_0 (Spawn%20Spline%20Mesh :SplineMeshOverrideDescriptions [(struct :InputSource StaticMesh :PropertyTarget "StaticMesh")] :SplineMeshParams (struct :bSmoothInterpRollScale false) :SplineMeshParamsOverride [(struct :InputSource @Data.RollDegrees :PropertyTarget "StartRollDegrees") (struct :InputSource @Data.RollDegreesEnd :PropertyTarget "EndRollDegrees") (struct :InputSource @Data.SplineUpDir :PropertyTarget "SplineUpDir")]))
  (bind SpawnSplineMesh_1 (Spawn%20Spline%20Mesh :SplineMeshOverrideDescriptions [(struct :InputSource StaticMesh :PropertyTarget "StaticMesh")] :SplineMeshParams (struct :bSmoothInterpRollScale false)))
  ;[0 0 0 0.5] Spline Mesh Y Forward axis rotation adjustment to match module standard
  (bind AttributeRotatorOp_0 (Attribute%20Rotator%20Op :InputSource1 $LocalRotation :InputSource2 $LocalRotation :OutputTarget $Rotation))
  (bind AttributeRotatorOp_1 (Attribute%20Rotator%20Op :InputSource1 $LocalRotation :InputSource2 $LocalRotation :OutputTarget $Rotation))
  (connect AttributeMathsOp_1 Out AttributeCompareOp_0 InA)
  (connect ExtractAttribute_0 Out AttributeMathsOp_0 InA)
  (connect AttributeMathsOp_0 Out AttributeMathsOp_1 In)
  (connect Branch_0 Output%20B AttributeMathsOp_4 InA)
  (connect AttributeMathsOp_4 Out AttributeMathsOp_5 InA)
  (connect Branch_3 Output%20A AttributeRotatorOp_0 InB)
  (connect Branch_3 Output%20B AttributeRotatorOp_1 InB)
  (connect AttributeRotatorOp_0 Out Branch_0 In)
  (connect AttributeRotatorOp_1 Out Branch_0 In)
  (connect Branch_1 Output%20A Branch_0 In)
  (connect AttributeCompareOp_0 Out Branch_1 bOutputToB)
  (connect Branch_2 Output%20A Branch_1 In)
  (connect SplineDirection_1 Out Branch_1 In)
  (connect ExtractAttribute_1 Out Branch_2 bOutputToB)
  (connect DefaultInputNode In Branch_2 In)
  (connect ExtractAttribute_1 Out Branch_3 bOutputToB)
  (connect Branch_1 Output%20B Branch_3 In)
  (connect DefaultInputNode In ExtractAttribute_0 In)
  (connect DefaultInputNode In ExtractAttribute_1 In)
  (connect AttributeMathsOp_1 Out SpawnSplineMesh_0 ForwardAxis)
  (connect AttributeMathsOp_5 Out SpawnSplineMesh_0 In)
  (connect AttributeMathsOp_1 Out SpawnSplineMesh_1 ForwardAxis)
  (connect Branch_0 Output%20A SpawnSplineMesh_1 In)
  (connect Branch_2 Output%20B SplineDirection_1 In))
